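; Quest gameplay ability that credits the player for the Operation Snowdown stoke-a-campfire objective.
; It responds only to events whose target tags pass a comparison against its campfire tag container.
; On activation it runs the parent logic, casts the target to the environment campfire and waits for a Light gameplay event on it.
; When that event arrives, it sends the stoke stat event to the owning player's quest manager and ends.
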(event Ability|EventActivateAbilityFromEvent
  (params (EventData :type Gameplay%20Event%20Data%20Structure%20%28by%20ref%29))
  (bind (_eventtag _instigator _target) (Utilities|Struct|BreakGameplayEventData EventData))
  (|CallParentFunction EventData)
  (bind _asb_bga_athena_env_camp_fire (Utilities|Casting|CastToB_BGA_Athena_EnvCampFire _target))
  (Ability|Tasks|WaitGameplayEvent "(TagName=\"Event.Athena.EnvCampFire.Light\")" _asb_bga_athena_env_camp_fire
    (:EventReceived
      (Quest|SendComplexCustomStatEvent (Variables|GAVars|Data|GetOwningPlayerQuestManager) 0 "(GameplayTags=((TagName=\"Athena.Quests.Stoke.Campfire\")))" "(GameplayTags=)" 1)
      (Ability|EndAbility))))

(fn ShouldAbilityRespondToEvent
  (params (ActorInfo :type Gameplay%20Ability%20Actor%20Info%20Structure) (Payload_EventTag :type Gameplay%20Tag%20Structure) (Payload_Instigator :type Actor%20Object%20Reference) (Payload_Target :type Actor%20Object%20Reference) (Payload_OptionalObject :type Object%20Reference) (Payload_OptionalObject2 :type Object%20Reference) (Payload_ContextHandle :type Gameplay%20Effect%20Context%20Handle%20Structure) (Payload_InstigatorTags :type Gameplay%20Tag%20Container%20Structure) (Payload_TargetTags :type Gameplay%20Tag%20Container%20Structure) (Payload_EventMagnitude :type Float%20%28single-precision%29) (Payload_TargetData :type Gameplay%20Ability%20Target%20Data%20Handle%20Structure) (Payload_InstancedEventData :type Instanced%20Struct%20Structure))
  (return (GameplayTags|TagContainer|CompareTagContainertoOtherTags (Variables|Default|GetT_Campfire) Payload_TargetTags)))
